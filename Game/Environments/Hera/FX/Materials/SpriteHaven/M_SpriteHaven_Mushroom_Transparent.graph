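# Translucent, unlit, two-sided material for Niagara mesh particles, named for the mushroom in the Hera SpriteHaven effects.
# Particle color RGB is multiplied by a saturated, powered Fresnel remapped through a lerp, so brightness varies with view angle.
# Particle alpha feeds a textureless Bayer dither function, a screen-door style fade technique alongside the translucent blend mode.
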Material /Game/Environments/Hera/FX/Materials/SpriteHaven/M_SpriteHaven_Mushroom_Transparent
graph {
 properties {
  bAutomaticallySetUsageInEditor=false
  bUsedWithNiagaraMeshParticles=true
  bUsedWithStaticMesh=false
  blendMode=BLEND_Translucent
  shadingModel=MSM_Unlit
  twoSided=true
  usedShadingModels=Unlit
 }
 nodes [
  n1 = ParticleColor()
  MaterialFunctionCall!(In=@n1.A, Iterations, Scale) {materialFunction=/Game/Packages/Fortress_SharedMaterials/Base_Material_Functions/MF_DitherBayerNoTex}
  n2 = Power(Base=Fresnel()) {constExponent=4}
  Multiply(A=@n1.RGB, B=LinearInterpolate(Alpha=Saturate(None=@n2)) {constA=0.5})
  Multiply()
 ]
}
